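# Physics asset for dynamic left-leg accessories on the medium male Garbage player body: an ankle bandanna and hanging scissors.
# Every body is a small capsule, with foot_l and calf_twist_02_l serving as the parents for the accessory bones.
# Both bandanna bones attach separately to the foot with swing cones and narrow twist limits.
# The scissors form a two-link chain from the calf twist bone, hinging on one swing axis with twist locked.
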PhysicsAsset /Game/Characters/Player/Male/Medium/Bodies/M_MED_Garbage/Meshes/M_MED_Garbage_Dyn_leg_l
bodies {
 ankle_bandanna_01 [{shapeType=Capsule, length=2, radius=2, center=<0.1 0 0>, rotation=<rot 90 0 0>}]
 ankle_bandanna_02 [{shapeType=Capsule, length=2, radius=2, center=<0.1 0 0>, rotation=<rot 90 0 0>}]
 calf_twist_02_l [{shapeType=Capsule, length=2, radius=2}]
 foot_l [{shapeType=Capsule, length=2, radius=2}]
 scissors_01 [{shapeType=Capsule, length=2, radius=2, center=<0.2 0 0>, rotation=<rot 90 0 0>}]
 scissors_02 [{shapeType=Capsule, length=2, radius=2, center=<0.2 0 0>, rotation=<rot 90 0 0>}]
}
constraints [
 {bone1Name=ankle_bandanna_01, bone2Name=foot_l, swing1LimitDegrees=45, swing2LimitDegrees=45, twistLimitDegrees=5}
 {bone1Name=ankle_bandanna_02, bone2Name=foot_l, swing1LimitDegrees=45, swing2LimitDegrees=45, twistLimitDegrees=5}
 {
  bone1Name=scissors_02
  bone2Name=scissors_01
  swing1LimitDegrees=45
  swing2LimitDegrees=0
  swing2Motion=Locked
  twistLimitDegrees=0
  twistMotion=Locked
 }
 {
  bone1Name=scissors_01
  bone2Name=calf_twist_02_l
  swing1LimitDegrees=45
  swing2LimitDegrees=0
  swing2Motion=Locked
  twistLimitDegrees=0
  twistMotion=Locked
 }
]
