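# Physics asset for the right-leg dynamic parts of the male medium Garbage player body, simulating a fish attached to the lower leg.
# Small capsule bodies cover the right calf twist bone and the fish head, body and tail.
# Head and body each constrain directly to the calf, while the tail chains off the body.
# All three joints share moderate swing limits and a tight twist limit, so the fish can swing with little twisting.
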PhysicsAsset /Game/Characters/Player/Male/Medium/Bodies/M_MED_Garbage/Meshes/M_MED_Garbage_Dyn_leg_r
bodies {
 calf_twist_02_r [{shapeType=Capsule, length=2, radius=2, rotation=<rot 90 0 0>}]
 fish_body [{shapeType=Capsule, length=2, radius=2, center=<0.1 0 0>, rotation=<rot 90 0 0>}]
 fish_head [{shapeType=Capsule, length=2, radius=2, center=<0.1 0 0>, rotation=<rot 90 0 0>}]
 fish_tail [{shapeType=Capsule, length=2, radius=2, center=<0.1 0 0>, rotation=<rot 90 0 0>}]
}
constraints [
 {bone1Name=fish_tail, bone2Name=fish_body, swing1LimitDegrees=45, swing2LimitDegrees=45, twistLimitDegrees=5}
 {bone1Name=fish_head, bone2Name=calf_twist_02_r, swing1LimitDegrees=45, swing2LimitDegrees=45, twistLimitDegrees=5}
 {bone1Name=fish_body, bone2Name=calf_twist_02_r, swing1LimitDegrees=45, swing2LimitDegrees=45, twistLimitDegrees=5}
]
