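# Physics asset for the medium male Garbage body, covering props that dangle from the pelvis, such as cans, matches, hammer, cup, heart and tp roll.
# Each prop hangs from the pelvis by its own constraint chain, some through rings, lids or a handle.
# The tp roll segments lock twist and one swing axis to move in a single plane.
# Pelvis and left thigh supply larger body shapes, with the thigh's rotation locked to the pelvis.
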PhysicsAsset /Game/Characters/Player/Male/Medium/Bodies/M_MED_Garbage/Meshes/M_MED_Garbage_Dyn_pelvis
bodies {
 can_bk [{shapeType=Capsule, length=2, radius=2, center=<0.2 0 0>, rotation=<rot 90 0 0>}]
 can_bk_lid [{shapeType=Capsule, length=2, radius=2, center=<0.2 0 0>, rotation=<rot 90 0 0>}]
 can_fr [{shapeType=Capsule, length=2, radius=2, center=<0.2 0 0>, rotation=<rot 90 0 0>}]
 can_fr_lid [{shapeType=Capsule, length=2, radius=2, center=<1.485 0 0.0001>, rotation=<rot 90 0 0>}]
 cup [
  {shapeType=Capsule, length=8, radius=1.25, center=<5.8 -0.2904 0.0001>, rotation=<rot 90 0 0>}
  {shapeType=Capsule, length=0, radius=4.25, center=<11.6 -0.9271 1.3518>}
 ]
 cup_ring [{shapeType=Capsule, length=2, radius=2, center=<0.7952 0 0>, rotation=<rot 90 0 0>}]
 hammer [
  {shapeType=Capsule, length=16.25, radius=1.5, center=<9.9851 0 0.0001>, rotation=<rot 90 0 0>}
  {shapeType=Capsule, length=9, radius=2.5, center=<17.7175 0.5876 0.0003>, rotation=<rot 90 -90 -179.9998>}
 ]
 hammer_ring [{shapeType=Capsule, length=2, radius=2, center=<0.8342 0 0>, rotation=<rot 90 0 0>}]
 heart [{shapeType=Box, extentX=7.5, extentY=7.5, extentZ=1.5, center=<5.018 0 -0.0001>}]
 heart_ring_01 [{shapeType=Capsule, length=0.5, radius=2, center=<0.6827 0 0>, rotation=<rot 90 0 0>}]
 heart_ring_02 [{shapeType=Capsule, length=0.75, radius=1.25, center=<0.7847 0 0>, rotation=<rot 90 0 0>}]
 matches [{shapeType=Capsule, length=2, radius=2, center=<0.2 0 0>, rotation=<rot 90 0 0>}]
 pelvis [
  {shapeType=Capsule, length=11, radius=13, center=<-2.3112 -0.4826 -1.862>}
  {shapeType=Box, extentX=12, extentY=7.5, extentZ=11.5, center=<-0.8742 -11.0072 -0.9055>, rotation=<rot 0 5.0007 0>}
 ]
 popcan [{shapeType=Capsule, length=2, radius=2, center=<0.2 0 0>, rotation=<rot 90 0 0>}]
 thigh_l [{shapeType=Capsule, length=16.9444, radius=9.3591, center=<-15.3438 0.0005 -2.2272>, rotation=<rot 90 0 0>}]
 tp_01 [{shapeType=Capsule, length=2, radius=2, center=<1 0 0>, rotation=<rot 90 0 0>}]
 tp_02 [{shapeType=Capsule, length=2, radius=2, center=<1 0 0>, rotation=<rot 90 0 0>}]
 tp_handle [{shapeType=Capsule, length=2, radius=2, center=<0.1 0 0>, rotation=<rot 90 0 0>}]
 tp_roll [{shapeType=Capsule, length=7.0651, radius=4.324, rotation=<rot 0 0 105.0005>}]
}
constraints [
 {bone1Name=tp_handle, bone2Name=pelvis, swing1LimitDegrees=1, swing2LimitDegrees=70, twistLimitDegrees=1}
 {
  bone1Name=tp_roll
  bone2Name=tp_handle
  swing1LimitDegrees=2
  swing1Motion=Locked
  swing2LimitDegrees=20
  twistLimitDegrees=2
  twistMotion=Locked
 }
 {
  bone1Name=tp_01
  bone2Name=tp_roll
  swing1LimitDegrees=2
  swing1Motion=Locked
  swing2LimitDegrees=20
  twistLimitDegrees=2
  twistMotion=Locked
 }
 {bone1Name=tp_02, bone2Name=tp_01, swing1LimitDegrees=2, swing1Motion=Locked, swing2LimitDegrees=45, twistLimitDegrees=2, twistMotion=Locked}
 {bone1Name=popcan, bone2Name=pelvis, swing1LimitDegrees=25, swing2LimitDegrees=45, twistLimitDegrees=5}
 {bone1Name=matches, bone2Name=pelvis, swing1LimitDegrees=5, swing2LimitDegrees=45, twistLimitDegrees=5}
 {bone1Name=can_fr, bone2Name=pelvis, swing1LimitDegrees=5, swing2LimitDegrees=45, twistLimitDegrees=5}
 {bone1Name=can_fr_lid, bone2Name=can_fr, swing1LimitDegrees=45, swing2LimitDegrees=2, twistLimitDegrees=5}
 {bone1Name=can_bk, bone2Name=pelvis, swing1LimitDegrees=15, swing2LimitDegrees=45, twistLimitDegrees=5}
 {bone1Name=can_bk_lid, bone2Name=can_bk, swing1LimitDegrees=2, swing2LimitDegrees=45, twistLimitDegrees=5}
 {bone1Name=hammer_ring, bone2Name=pelvis, swing1LimitDegrees=15, swing2LimitDegrees=45, twistLimitDegrees=5}
 {bone1Name=hammer, bone2Name=hammer_ring, swing1LimitDegrees=15, swing2LimitDegrees=45, twistLimitDegrees=5}
 {bone1Name=cup_ring, bone2Name=pelvis, swing1LimitDegrees=15, swing2LimitDegrees=45, twistLimitDegrees=5}
 {bone1Name=cup, bone2Name=cup_ring, swing1LimitDegrees=15, swing2LimitDegrees=45, twistLimitDegrees=5}
 {bone1Name=heart_ring_01, bone2Name=pelvis, swing1LimitDegrees=15, swing2LimitDegrees=45, twistLimitDegrees=5}
 {bone1Name=heart_ring_02, bone2Name=heart_ring_01, swing1LimitDegrees=15, swing2LimitDegrees=45, twistLimitDegrees=5}
 {bone1Name=heart, bone2Name=heart_ring_02, swing1LimitDegrees=15, swing2LimitDegrees=45, twistLimitDegrees=5}
 {
  bone1Name=thigh_l
  bone2Name=pelvis
  swing1LimitDegrees=45
  swing1Motion=Locked
  swing2LimitDegrees=45
  swing2Motion=Locked
  twistLimitDegrees=45
  twistMotion=Locked
 }
]
