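# Physics asset for the banana on the left shoulder pad of the medium male Garbage player body, named as a dynamic "perf" variant.
# A two-capsule banana base-to-top chain and four box peels hang off the spherical shoulder pad body with limited swing and twist.
# The pad links through the upper-arm twist to the shoulder cup with free angular motion.
# An unconstrained capsule-and-box spine body covers the torso.
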PhysicsAsset /Game/Characters/Player/Male/Medium/Bodies/M_MED_Garbage/Meshes/M_MED_Garbage_Dyn_perf_banana
bodies {
 alt_upperarm_twist_01_l [{shapeType=Capsule, length=2, radius=2}]
 banana_base [{shapeType=Capsule, length=1.0474, radius=2.75, center=<2 0 0>, rotation=<rot 90 0 0>}]
 banana_peel_01 [{shapeType=Box, extentX=4.7718, extentY=1.0501, extentZ=3.0862, center=<2.922 0 0.0002>}]
 banana_peel_02 [{shapeType=Box, extentX=4.7718, extentY=2.0501, extentZ=3.0862, center=<2.922 0 0.0002>}]
 banana_peel_03 [{shapeType=Box, extentX=4.7718, extentY=1.3001, extentZ=3.0862, center=<2.922 0 0.0002>}]
 banana_peel_04 [{shapeType=Box, extentX=4.7718, extentY=1.0501, extentZ=3.0862, center=<2.922 0 0.0002>}]
 banana_top [{shapeType=Capsule, length=3.75, radius=2.25, center=<0.1 0 0>, rotation=<rot 90 0 0>}]
 dyn_shoulder_cup_l [{shapeType=Capsule, length=2, radius=2}]
 dyn_spine_04 [
  {shapeType=Capsule, length=3.75, radius=8.25}
  {shapeType=Box, extentX=28.2367, extentY=18.1469, extentZ=33.5002, center=<5.683 -4.598 -2.0815>, rotation=<rot 0 3.5623 0>}
 ]
 shoulderpad_l [{shapeType=Sphere, radius=10.8352, center=<8.5041 -0.0412 -3.5825>}]
}
constraints [
 {bone1Name=banana_top, bone2Name=banana_base, swing1LimitDegrees=30, swing2LimitDegrees=30, twistLimitDegrees=10}
 {bone1Name=banana_base, bone2Name=shoulderpad_l, swing1LimitDegrees=30, swing2LimitDegrees=30, twistLimitDegrees=10}
 {bone1Name=banana_peel_01, bone2Name=shoulderpad_l, swing1LimitDegrees=30, swing2LimitDegrees=15, twistLimitDegrees=30}
 {bone1Name=banana_peel_02, bone2Name=shoulderpad_l, swing1LimitDegrees=30, swing2LimitDegrees=15, twistLimitDegrees=30}
 {bone1Name=banana_peel_03, bone2Name=shoulderpad_l, swing1LimitDegrees=30, swing2LimitDegrees=15, twistLimitDegrees=30}
 {bone1Name=banana_peel_04, bone2Name=shoulderpad_l, swing1LimitDegrees=30, swing2LimitDegrees=15, twistLimitDegrees=30}
 {
  bone1Name=shoulderpad_l
  bone2Name=alt_upperarm_twist_01_l
  swing1LimitDegrees=45
  swing1Motion=Free
  swing2LimitDegrees=45
  swing2Motion=Free
  twistLimitDegrees=45
  twistMotion=Free
 }
 {
  bone1Name=alt_upperarm_twist_01_l
  bone2Name=dyn_shoulder_cup_l
  swing1LimitDegrees=45
  swing1Motion=Free
  swing2LimitDegrees=45
  swing2Motion=Free
  twistLimitDegrees=45
  twistMotion=Free
 }
]
