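# Physics asset for the left shoulder of the medium male player Garbage body, covering the dynamic shoulder pad and flashlight.
# The pad has limited swing and twist against its MSTR bone, and the flashlight has tighter limits against the second upper-arm twist bone.
# Spine, clavicle, shoulder-cup and arm bodies link through rotationally free constraints, except dyn_spine_04, locked to spine_04.
# Boxes shape the pad, flashlight and dynamic spine; spheres and capsules cover the rest.
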PhysicsAsset /Game/Characters/Player/Male/Medium/Bodies/M_MED_Garbage/Meshes/M_MED_Garbage_Dyn_shoulder_l
bodies {
 alt_upperarm_twist_01_l [
  {shapeType=Sphere, radius=6.3901}
  {shapeType=Capsule, length=11.8304, radius=6.8924, center=<14.159 1.9414 -0.709>, rotation=<rot 86.8781 -179.9999 -179.9999>}
 ]
 clavicle_l [{shapeType=Capsule, length=1, radius=1}]
 dyn_shoulder_cup_l [{shapeType=Capsule, length=2, radius=2}]
 dyn_spine_04 [{shapeType=Box, extentX=28.2367, extentY=18.1469, extentZ=30.5, center=<5.6829 -4.5974 0.0001>, rotation=<rot 0 3.5623 0>}]
 flashlight [{shapeType=Box, extentX=2.8787, extentY=7.5, extentZ=5.8922, center=<-6.0322 0 3.0294>, rotation=<rot -1.3453 0 0>}]
 shoulderpad_l [{shapeType=Box, extentX=15.8739, extentY=12.6604, extentZ=2.1161, center=<10 -2 6>, rotation=<rot 5.0014 -90.0002 1.5106>}]
 shoulderpad_l_MSTR [{shapeType=Capsule, length=0.505, radius=0.505}]
 spine_04 [{shapeType=Capsule, length=3, radius=3}]
 upperarm_twist_01_l [{shapeType=Capsule, length=7.2391, radius=0.5863, center=<13.0249 1.4178 2.2261>, rotation=<rot 56.9439 28.2094 119.3825>}]
 upperarm_twist_02_l [{shapeType=Capsule, length=1, radius=1, rotation=<rot 90 0 0>}]
}
constraints [
 {bone1Name=flashlight, bone2Name=upperarm_twist_02_l, swing1LimitDegrees=20, swing2LimitDegrees=15, twistLimitDegrees=3}
 {bone1Name=shoulderpad_l, bone2Name=shoulderpad_l_MSTR, swing1LimitDegrees=15, swing2LimitDegrees=40, twistLimitDegrees=15}
 {
  bone1Name=upperarm_twist_02_l
  bone2Name=clavicle_l
  swing1LimitDegrees=45
  swing1Motion=Free
  swing2LimitDegrees=45
  swing2Motion=Free
  twistLimitDegrees=45
  twistMotion=Free
 }
 {
  bone1Name=dyn_spine_04
  bone2Name=spine_04
  swing1LimitDegrees=45
  swing1Motion=Locked
  swing2LimitDegrees=45
  swing2Motion=Locked
  twistLimitDegrees=45
  twistMotion=Locked
 }
 {
  bone1Name=upperarm_twist_01_l
  bone2Name=clavicle_l
  swing1LimitDegrees=45
  swing1Motion=Free
  swing2LimitDegrees=45
  swing2Motion=Free
  twistLimitDegrees=45
  twistMotion=Free
 }
 {
  bone1Name=dyn_shoulder_cup_l
  bone2Name=clavicle_l
  swing1LimitDegrees=45
  swing1Motion=Free
  swing2LimitDegrees=45
  swing2Motion=Free
  twistLimitDegrees=45
  twistMotion=Free
 }
 {
  bone1Name=alt_upperarm_twist_01_l
  bone2Name=dyn_shoulder_cup_l
  swing1LimitDegrees=45
  swing1Motion=Free
  swing2LimitDegrees=45
  swing2Motion=Free
  twistLimitDegrees=45
  twistMotion=Free
 }
 {
  bone1Name=clavicle_l
  bone2Name=spine_04
  swing1LimitDegrees=45
  swing1Motion=Free
  swing2LimitDegrees=45
  swing2Motion=Free
  twistLimitDegrees=45
  twistMotion=Free
 }
 {
  bone1Name=shoulderpad_l_MSTR
  bone2Name=dyn_shoulder_cup_l
  swing1LimitDegrees=45
  swing1Motion=Free
  swing2LimitDegrees=45
  swing2Motion=Free
  twistLimitDegrees=45
  twistMotion=Free
 }
]
