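# Physics asset for the torso dynamics of the male medium Garbage player body, covering shoulder pads, an antenna and a tire.
# Capsules, boxes and a sphere on spine, clavicle, arm and dyn bones form collision, linked mostly by fully free constraints back to spine_04.
# Both shoulder pads and the antenna have limited motion, the right pad on one swing axis only.
# The tire locks rigidly to dyn_spine_04, itself locked to spine_04.
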PhysicsAsset /Game/Characters/Player/Male/Medium/Bodies/M_MED_Garbage/Meshes/M_MED_Garbage_Dyn_torso
bodies {
 alt_upperarm_twist_01_l [
  {shapeType=Sphere, radius=6.3901}
  {shapeType=Capsule, length=0.4586, radius=5.6431, center=<6.8747 1.9415 -1.3813>, rotation=<rot 90 0 0>}
  {shapeType=Capsule, length=11.6911, radius=2.5843, center=<10.0284 1.2282 -0.3466>, rotation=<rot 80.2502 -180 -180>}
 ]
 alt_upperarm_twist_01_r [{shapeType=Capsule, length=2, radius=2, rotation=<rot 90 0 0>}]
 antenna [{shapeType=Capsule, length=7.5, radius=1, center=<0.2 0 0>, rotation=<rot 90 0 0>}]
 clavicle_l [{shapeType=Capsule, length=1, radius=1}]
 dyn_shoulder_cup_l [{shapeType=Capsule, length=2, radius=2}]
 dyn_shoulder_cup_r [{shapeType=Capsule, length=2, radius=2}]
 dyn_spine_04 [
  {shapeType=Capsule, length=3.75, radius=8.25}
  {shapeType=Box, extentX=28.2367, extentY=18.1469, extentZ=30.5, center=<5.6829 -4.5974 0.0001>, rotation=<rot 0 3.5623 0>}
 ]
 dyn_tire [{shapeType=Capsule, length=3.75, radius=8.25}]
 lowerarm_twist_02_r [{shapeType=Capsule, length=1, radius=1}]
 shoulderpad_l [{shapeType=Box, extentX=15.8739, extentY=12.6604, extentZ=2.1161, center=<10 -2 6>, rotation=<rot 5.0014 -90.0002 1.5106>}]
 shoulderpad_l_MSTR [{shapeType=Capsule, length=0.505, radius=0.505}]
 shoulderpad_r [{shapeType=Capsule, length=4.9132, radius=3.9007, center=<4.3835 0.0001 -0.3023>, rotation=<rot 84.9998 -0.0001 0>}]
 spine_04 [{shapeType=Capsule, length=3, radius=3}]
 upperarm_twist_01_l [
  {shapeType=Capsule, length=30.6409, radius=6.7282, center=<38.9607 -0.5365 2.4728>, rotation=<rot 83.4838 180 180>}
  {shapeType=Capsule, length=3.7458, radius=4.0172, center=<13.0249 1.4178 2.2261>, rotation=<rot 56.9439 28.2094 119.3825>}
 ]
 upperarm_twist_01_r [{shapeType=Capsule, length=4.3737, radius=5.6431, center=<-10.0786 -1.1848 1.4671>, rotation=<rot 90 0 0>}]
 upperarm_twist_02_l [{shapeType=Capsule, length=1, radius=1, rotation=<rot 90 0 0>}]
}
constraints [
 {bone1Name=shoulderpad_l, bone2Name=shoulderpad_l_MSTR, swing1LimitDegrees=15, swing2LimitDegrees=15, twistLimitDegrees=15}
 {
  bone1Name=upperarm_twist_02_l
  bone2Name=clavicle_l
  swing1LimitDegrees=45
  swing1Motion=Free
  swing2LimitDegrees=45
  swing2Motion=Free
  twistLimitDegrees=45
  twistMotion=Free
 }
 {
  bone1Name=dyn_spine_04
  bone2Name=spine_04
  swing1LimitDegrees=45
  swing1Motion=Locked
  swing2LimitDegrees=45
  swing2Motion=Locked
  twistLimitDegrees=45
  twistMotion=Locked
 }
 {
  bone1Name=shoulderpad_r
  bone2Name=alt_upperarm_twist_01_r
  swing1LimitDegrees=45
  swing1Motion=Locked
  swing2LimitDegrees=10
  twistLimitDegrees=45
  twistMotion=Locked
 }
 {
  bone1Name=upperarm_twist_01_r
  bone2Name=spine_04
  swing1LimitDegrees=45
  swing1Motion=Free
  swing2LimitDegrees=45
  swing2Motion=Free
  twistLimitDegrees=45
  twistMotion=Free
 }
 {
  bone1Name=upperarm_twist_01_l
  bone2Name=clavicle_l
  swing1LimitDegrees=45
  swing1Motion=Free
  swing2LimitDegrees=45
  swing2Motion=Free
  twistLimitDegrees=45
  twistMotion=Free
 }
 {
  bone1Name=dyn_tire
  bone2Name=dyn_spine_04
  swing1LimitDegrees=45
  swing1Motion=Locked
  swing2LimitDegrees=45
  swing2Motion=Locked
  twistLimitDegrees=45
  twistMotion=Locked
 }
 {bone1Name=antenna, bone2Name=lowerarm_twist_02_r, swing1LimitDegrees=45, swing2LimitDegrees=45, twistLimitDegrees=3}
 {
  bone1Name=lowerarm_twist_02_r
  bone2Name=spine_04
  swing1LimitDegrees=45
  swing1Motion=Free
  swing2LimitDegrees=45
  swing2Motion=Free
  twistLimitDegrees=45
  twistMotion=Free
 }
 {
  bone1Name=dyn_shoulder_cup_l
  bone2Name=clavicle_l
  swing1LimitDegrees=45
  swing1Motion=Free
  swing2LimitDegrees=45
  swing2Motion=Free
  twistLimitDegrees=45
  twistMotion=Free
 }
 {
  bone1Name=dyn_shoulder_cup_r
  bone2Name=spine_04
  swing1LimitDegrees=45
  swing1Motion=Free
  swing2LimitDegrees=45
  swing2Motion=Free
  twistLimitDegrees=45
  twistMotion=Free
 }
 {
  bone1Name=alt_upperarm_twist_01_l
  bone2Name=dyn_shoulder_cup_l
  swing1LimitDegrees=45
  swing1Motion=Free
  swing2LimitDegrees=45
  swing2Motion=Free
  twistLimitDegrees=45
  twistMotion=Free
 }
 {
  bone1Name=alt_upperarm_twist_01_r
  bone2Name=dyn_shoulder_cup_r
  swing1LimitDegrees=45
  swing1Motion=Free
  swing2LimitDegrees=45
  swing2Motion=Free
  twistLimitDegrees=45
  twistMotion=Free
 }
 {
  bone1Name=clavicle_l
  bone2Name=spine_04
  swing1LimitDegrees=45
  swing1Motion=Free
  swing2LimitDegrees=45
  swing2Motion=Free
  twistLimitDegrees=45
  twistMotion=Free
 }
 {
  bone1Name=shoulderpad_l_MSTR
  bone2Name=dyn_shoulder_cup_l
  swing1LimitDegrees=45
  swing1Motion=Free
  swing2LimitDegrees=45
  swing2Motion=Free
  twistLimitDegrees=45
  twistMotion=Free
 }
]
