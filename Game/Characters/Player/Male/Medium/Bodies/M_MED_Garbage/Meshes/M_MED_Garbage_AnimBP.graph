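# Post-process style Animation Blueprint for the M_MED_Garbage player body, copying the parent mesh's pose and adding secondary motion to its trash props.
# Constraints, Look Ats and bone offsets driven by arm/thigh angles, skydiving and a hands-replaced tag correct attachments.
# AnimDynamics and RigidBody nodes simulate dangling items; boot, armband, leg and left-shoulder simulations sit outside the output chain.
# Event Graph updates those drivers and resets dynamics around skydiving, parachute, targeting and crouch events.
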AnimBlueprint /Game/Characters/Player/Male/Medium/Bodies/M_MED_Garbage/Meshes/M_MED_Garbage_AnimBP
target_skeleton=&/M_MED_Garbage_Skeleton
graphs [
 AnimGraph {nodes [
   Anim Dynamics - Bone: dyn_boot_r!(ComponentPose, Alpha=0.5, GravityOverride:Vector="0,0,0", ExternalForce="0,0,0") -> (Pose) {
    angularDampingOverride=0.9999
    angularSpringConstant=1500
    bAngularSpring=true
    bGravityOverrideInSimSpace=true
    bOverrideAngularDamping=true
    gravityScale=0.8
    simulationSpace=RootRelative
    alphaBoolBlend {blendOption=Linear}
    boundBone {boneName=dyn_boot_r}
    componentAppliedLinearAccClamp=<5000 5000 3500>
    componentLinearAccScale=<1.5 1.5 1.25>
    componentLinearVelScale=<1.5 1.5 1.25>
    physicsBodyDefinitions [{
      sphereCollisionRadius=0
      boundBone {boneName=dyn_boot_r}
      boxExtents=<10 20 20>
      constraintSetup {angularLimitsMax=<15 35 15>, angularLimitsMin=<-15 -35 -15>, angularTarget=<0 0 0>}
      localJointOffset=<2 0 0>
     }]
    relativeSpaceBone {boneName=root}
   }
   n1 = Anim Dynamics - Bone: dyn_armband_b_01(Alpha=1, ExternalForce="0,0,0") {
    angularDampingOverride=0.9999
    angularSpringConstant=50
    bAngularSpring=true
    bGravityOverrideInSimSpace=true
    bOverrideAngularDamping=true
    bUseGravityOverride=true
    simulationSpace=RootRelative
    alphaBoolBlend {blendOption=Linear}
    boundBone {boneName=dyn_armband_b_01}
    componentAppliedLinearAccClamp=<3500 3500 3500>
    componentLinearAccScale=<0.8 0.8 0.4>
    componentLinearVelScale=<0.8 0.8 0.4>
    gravityOverride {z=-720}
    physicsBodyDefinitions [{
      sphereCollisionRadius=0
      boundBone {boneName=dyn_armband_b_01}
      boxExtents=<10 10 5>
      constraintSetup {angularLimitsMax=<10 30 25>, angularLimitsMin=<-10 -30 -5>, angularTarget=<0 0 0>}
      localJointOffset=<4 0 0>
     }]
    relativeSpaceBone {boneName=dyn_root_simspace}
   }
   Anim Dynamics - Bone: dyn_armband_b_02(ComponentPose=@n1, Alpha=1, ExternalForce="0,0,0") {
    angularDampingOverride=0.9999
    angularSpringConstant=50
    bAngularSpring=true
    bGravityOverrideInSimSpace=true
    bOverrideAngularDamping=true
    bUseGravityOverride=true
    simulationSpace=RootRelative
    alphaBoolBlend {blendOption=Linear}
    boundBone {boneName=dyn_armband_b_02}
    componentAppliedLinearAccClamp=<3500 3500 3500>
    componentLinearAccScale=<0.8 0.8 0.4>
    componentLinearVelScale=<0.8 0.8 0.4>
    gravityOverride {z=-720}
    physicsBodyDefinitions [{
      sphereCollisionRadius=0
      boundBone {boneName=dyn_armband_b_02}
      boxExtents=<10 10 5>
      constraintSetup {angularLimitsMax=<10 30 25>, angularLimitsMin=<-10 -30 -5>, angularTarget=<0 0 0>}
      localJointOffset=<4 0 0>
     }]
    relativeSpaceBone {boneName=dyn_root_simspace}
   }
   "Transform (Modify) Bone - Bone: shoulderpad_l_MSTR"(Alpha=1, Translation="0,0,0", Rotation=Get LeftShoulder_Rot() -> (LeftShoulder_Rot:Rotator), Scale="1,1,1") {rotationMode=BMM_Additive, rotationSpace=BCS_BoneSpace, alphaBoolBlend {blendOption=Linear}, boneToModify {boneName=shoulderpad_l_MSTR}}
   n2 = RigidBody() {
    bClampLinearTranslationLimitToRefPose=true
    bOverrideWorldGravity=true
    overridePhysicsAsset=&/M_MED_Garbage_Dyn_leg_l
    simulationSpace=BaseBoneSpace
    alphaBoolBlend {blendOption=Linear}
    baseBoneRef {boneName=thigh_l}
    componentAppliedLinearAccClamp=<3000 3000 3000>
    componentLinearAccScale=<0.6 0.6 0.6>
    componentLinearVelScale=<0.6 0.6 0.6>
    overrideWorldGravity {z=-980}
   }
   RigidBody(ComponentPose=@n2) {
    bClampLinearTranslationLimitToRefPose=true
    bOverrideWorldGravity=true
    overridePhysicsAsset=&/M_MED_Garbage_Dyn_leg_r
    simulationSpace=BaseBoneSpace
    alphaBoolBlend {blendOption=Linear}
    baseBoneRef {boneName=thigh_r}
    componentAppliedLinearAccClamp=<3000 3000 3000>
    componentLinearAccScale=<0.6 0.6 0.6>
    componentLinearVelScale=<0.6 0.6 0.6>
    overrideWorldGravity {z=-980}
   }
   RigidBody() {
    bClampLinearTranslationLimitToRefPose=true
    bOverrideWorldGravity=true
    overridePhysicsAsset=&/M_MED_Garbage_Dyn_shoulder_l
    simulationSpace=BaseBoneSpace
    alphaBoolBlend {blendOption=Linear}
    baseBoneRef {boneName=spine_04}
    componentAppliedLinearAccClamp=<3000 3000 3000>
    componentLinearAccScale=<0.6 0.6 0.6>
    componentLinearVelScale=<0.6 0.6 0.6>
    overrideWorldGravity {z=-980}
   }
   n3 = "Constraint - alt_upperarm_twist_01_l (R)"(ComponentPose=Local To Component(LocalPose=Copy Pose From Mesh() {bUseAttachedParent=true}), Alpha=1, ConstraintWeights_0=1) {
    alphaBoolBlend {blendOption=Linear}
    boneToModify {boneName=alt_upperarm_twist_01_l}
    constraintSetup [{transformType=Rotation, targetBone {boneName=upperarm_twist_01_l}}]
    constraintWeights [1]
   }
   n4 = "Constraint - alt_upperarm_twist_01_r (R)"(ComponentPose=@n3, Alpha=1, ConstraintWeights_0=1) {
    alphaBoolBlend {blendOption=Linear}
    boneToModify {boneName=alt_upperarm_twist_01_r}
    constraintSetup [{transformType=Rotation, targetBone {boneName=upperarm_twist_01_r}}]
    constraintWeights [1]
   }
   n5 = "Constraint - dyn_shoulder_cup_l (R)"(ComponentPose=@n4, Alpha=1, ConstraintWeights_0=1) {
    alphaBoolBlend {blendOption=Linear}
    boneToModify {boneName=dyn_shoulder_cup_l}
    constraintSetup [{transformType=Rotation, targetBone {boneName=spine_04}}]
    constraintWeights [1]
   }
   n6 = "Constraint - dyn_shoulder_cup_r (R)"(ComponentPose=@n5, Alpha=1, ConstraintWeights_0=1) {
    alphaBoolBlend {blendOption=Linear}
    boneToModify {boneName=dyn_shoulder_cup_r}
    constraintSetup [{transformType=Rotation, targetBone {boneName=spine_04}}]
    constraintWeights [1]
   }
   n7 = Look At - Bone: alt_upperarm_twist_01_l(ComponentPose=@n6, Alpha=1) {
    alphaBoolBlend {blendOption=Linear}
    boneToModify {boneName=alt_upperarm_twist_01_l}
    lookAtLocation {x=0}
    lookAtTarget {boneReference {boneName=lowerarm_l}}
    lookAt_Axis {axis=<1 0>}
   }
   n8 = Look At - Bone: alt_upperarm_twist_01_r(ComponentPose=@n7, Alpha=1) {
    alphaBoolBlend {blendOption=Linear}
    boneToModify {boneName=alt_upperarm_twist_01_r}
    lookAtLocation {x=0}
    lookAtTarget {boneReference {boneName=lowerarm_r}}
    lookAt_Axis {axis=<-1 0>}
   }
   n9 = Look At - Bone: shoulderpad_l_MSTR(ComponentPose=@n8, Alpha=0.5) {
    alphaBoolBlend {blendOption=Linear}
    boneToModify {boneName=shoulderpad_l_MSTR}
    lookAtLocation {x=0, z=15}
    lookAtTarget {boneReference {boneName=lowerarm_l}}
    lookAt_Axis {axis=<1 0>}
    lookUp_Axis {axis {x=0, z=1}}
   }
   n10 = Anim Dynamics - Bone: ankle_bandanna_01(ComponentPose=@n9, Alpha=1, ExternalForce="0,0,0") {
    angularDampingOverride=0.998
    bOverrideAngularDamping=true
    simulationSpace=RootRelative
    alphaBoolBlend {blendOption=Linear}
    boundBone {boneName=ankle_bandanna_01}
    componentLinearAccScale=<0.8 0.8 0.8>
    componentLinearVelScale=<0.8 0.8 0.8>
    physicsBodyDefinitions [{
      boundBone {boneName=ankle_bandanna_01}
      boxExtents=<7 12 12>
      constraintSetup {angularLimitsMax=<45 45 45>, angularLimitsMin=<-45 -45 -45>, angularTarget=<0 0 0>}
      localJointOffset=<2 0 0>
     }]
   }
   n11 = Anim Dynamics - Bone: ankle_bandanna_02(ComponentPose=@n10, Alpha=1, ExternalForce="0,0,0") {
    angularDampingOverride=0.998
    bOverrideAngularDamping=true
    simulationSpace=RootRelative
    alphaBoolBlend {blendOption=Linear}
    boundBone {boneName=ankle_bandanna_02}
    componentLinearAccScale=<0.8 0.8 0.8>
    componentLinearVelScale=<0.8 0.8 0.8>
    physicsBodyDefinitions [{
      boundBone {boneName=ankle_bandanna_02}
      boxExtents=<7 12 12>
      constraintSetup {angularLimitsMax=<45 45 45>, angularLimitsMin=<-45 -45 -45>, angularTarget=<0 0 0>}
      localJointOffset=<2 0 0>
     }]
   }
   n12 = Anim Dynamics - Bone: scissors_01(ComponentPose=@n11, Alpha=1, ExternalForce="0,0,0") {
    angularDampingOverride=0.998
    bOverrideAngularDamping=true
    simulationSpace=RootRelative
    alphaBoolBlend {blendOption=Linear}
    boundBone {boneName=scissors_01}
    componentLinearAccScale=<0.8 0.8 0.8>
    componentLinearVelScale=<0.8 0.8 0.8>
    physicsBodyDefinitions [{
      boundBone {boneName=scissors_01}
      boxExtents=<10 15 15>
      constraintSetup {angularLimitsMax=<1 1 15>, angularLimitsMin=<-1 -1 -15>, angularTarget=<0 0 0>}
      localJointOffset=<2 0 0>
     }]
   }
   n13 = Anim Dynamics - Bone: scissors_02(ComponentPose=@n12, Alpha=1, ExternalForce="0,0,0") {
    angularDampingOverride=0.998
    bOverrideAngularDamping=true
    simulationSpace=RootRelative
    alphaBoolBlend {blendOption=Linear}
    boundBone {boneName=scissors_02}
    componentLinearAccScale=<0.8 0.8 0.8>
    componentLinearVelScale=<0.8 0.8 0.8>
    physicsBodyDefinitions [{
      boundBone {boneName=scissors_02}
      boxExtents=<7 12 12>
      constraintSetup {angularLimitsMax=<1 1 10>, angularLimitsMin=<-1 -1 -10>, angularTarget=<0 0 0>}
      localJointOffset=<2 0 0>
     }]
   }
   n14 = Anim Dynamics - Bone: fish_head(ComponentPose=@n13, Alpha=1, ExternalForce="0,0,0") {
    angularDampingOverride=0.998
    bOverrideAngularDamping=true
    simulationSpace=RootRelative
    alphaBoolBlend {blendOption=Linear}
    boundBone {boneName=fish_head}
    componentLinearAccScale=<0.8 0.8 0.8>
    componentLinearVelScale=<0.8 0.8 0.8>
    physicsBodyDefinitions [{
      boundBone {boneName=fish_head}
      boxExtents=<7 12 12>
      constraintSetup {angularLimitsMax=<1 1 10>, angularLimitsMin=<-1 -1 -10>, angularTarget=<0 0 0>}
      localJointOffset=<2 0 0>
     }]
   }
   n15 = Anim Dynamics - Bone: fish_body(ComponentPose=@n14, Alpha=1, ExternalForce="0,0,0") {
    angularDampingOverride=0.998
    bOverrideAngularDamping=true
    simulationSpace=RootRelative
    alphaBoolBlend {blendOption=Linear}
    boundBone {boneName=fish_body}
    componentLinearAccScale=<0.8 0.8 0.8>
    componentLinearVelScale=<0.8 0.8 0.8>
    physicsBodyDefinitions [{
      boundBone {boneName=fish_body}
      boxExtents=<10 12 12>
      constraintSetup {angularLimitsMax=<1 1 20>, angularLimitsMin=<-1 -1 -20>, angularTarget=<0 0 0>}
      localJointOffset=<2 0 0>
     }]
   }
   n16 = Anim Dynamics - Bone: fish_tail(ComponentPose=@n15, Alpha=1, ExternalForce="0,0,0") {
    angularDampingOverride=0.998
    bOverrideAngularDamping=true
    simulationSpace=RootRelative
    alphaBoolBlend {blendOption=Linear}
    boundBone {boneName=fish_tail}
    componentLinearAccScale=<0.8 0.8 0.8>
    componentLinearVelScale=<0.8 0.8 0.8>
    physicsBodyDefinitions [{
      boundBone {boneName=fish_tail}
      boxExtents=<10 12 12>
      constraintSetup {angularLimitsMax=<1 1 15>, angularLimitsMin=<-1 -1 -15>, angularTarget=<0 0 0>}
      localJointOffset=<2 0 0>
     }]
   }
   n17 = Anim Dynamics - Bone: wire_01(ComponentPose=@n16, Alpha=1, ExternalForce="0,0,0") {
    angularDampingOverride=0.998
    bOverrideAngularDamping=true
    simulationSpace=RootRelative
    alphaBoolBlend {blendOption=Linear}
    boundBone {boneName=wire_01}
    componentLinearAccScale=<0.8 0.8 0.8>
    componentLinearVelScale=<0.8 0.8 0.8>
    physicsBodyDefinitions [{
      boundBone {boneName=wire_01}
      boxExtents=<5 10 10>
      constraintSetup {angularLimitsMax=<1 25 25>, angularLimitsMin=<-1 -25 -25>, angularTarget=<0 0 0>}
      localJointOffset=<2 0 0>
     }]
   }
   n18 = Anim Dynamics - Bone: wire_02(ComponentPose=@n17, Alpha=1, ExternalForce="0,0,0") {
    angularDampingOverride=0.998
    bOverrideAngularDamping=true
    simulationSpace=RootRelative
    alphaBoolBlend {blendOption=Linear}
    boundBone {boneName=wire_02}
    componentLinearAccScale=<0.8 0.8 0.8>
    componentLinearVelScale=<0.8 0.8 0.8>
    physicsBodyDefinitions [{
      boundBone {boneName=wire_02}
      boxExtents=<5 10 10>
      constraintSetup {angularLimitsMax=<1 25 25>, angularLimitsMin=<-1 -25 -25>, angularTarget=<0 0 0>}
      localJointOffset=<2 0 0>
     }]
   }
   n19 = Anim Dynamics - Bone: antenna(ComponentPose=@n18, Alpha=1, ExternalForce="0,0,0") {
    angularDampingOverride=0.998
    angularSpringConstant=1500
    bAngularSpring=true
    bOverrideAngularDamping=true
    bUseGravityOverride=true
    simulationSpace=RootRelative
    alphaBoolBlend {blendOption=Linear}
    boundBone {boneName=antenna}
    componentLinearAccScale=<0.8 0.8 0.8>
    componentLinearVelScale=<0.8 0.8 0.8>
    gravityOverride {z=-540}
    physicsBodyDefinitions [{
      boundBone {boneName=antenna}
      boxExtents=<10 15 15>
      constraintSetup {angularLimitsMax=<1 25 25>, angularLimitsMin=<-1 -25 -25>, angularTarget=<0 0 0>}
      localJointOffset=<1 0 0>
     }]
   }
   n20 = Get left arm rot() -> (left arm rot:double)
   n21 = "Transform (Modify) Bone - Bone: shoulderpad_l_MSTR"(ComponentPose=@n19, Alpha=@n20, Translation="-1.234,-10.016,-1.969", Rotation="6.529,21.445,-45.299", Scale="1,1,1") {
    rotationMode=BMM_Additive
    rotationSpace=BCS_BoneSpace
    translationMode=BMM_Additive
    translationSpace=BCS_BoneSpace
    alphaBoolBlend {blendOption=Linear}
    boneToModify {boneName=shoulderpad_l_MSTR}
    rotation=<rot 6.529 21.445 -45.299>
    translation=<-1.234 -10.016 -1.969>
   }
   n22 = "Transform (Modify) Bone - Bone: flashlight"(ComponentPose=@n21, Alpha=@n20, Translation="0.027,0.032,1", Rotation="-24.869,-5.442,1.966", Scale="1,1,1") {
    rotationMode=BMM_Additive
    rotationSpace=BCS_BoneSpace
    translationMode=BMM_Additive
    translationSpace=BCS_BoneSpace
    alphaBoolBlend {blendOption=Linear}
    boneToModify {boneName=flashlight}
    rotation=<rot -24.869 -5.442 1.9661>
    translation=<0.027 0.032 1>
   }
   n23 = Anim Dynamics - Bone: flashlight(ComponentPose=@n22, Alpha=1, ExternalForce="0,0,0") {
    angularDampingOverride=0.998
    bOverrideAngularDamping=true
    simulationSpace=RootRelative
    alphaBoolBlend {blendOption=Linear}
    boundBone {boneName=flashlight}
    componentLinearAccScale=<0.4 0.4 0.4>
    componentLinearVelScale=<0.4 0.4 0.4>
    physicsBodyDefinitions [{
      boundBone {boneName=flashlight}
      boxExtents=<15 20 20>
      constraintSetup {angularLimitsMax=<1 10 5>, angularLimitsMin=<-1 -10 -5>, angularTarget=<0 0 0>}
      localJointOffset=<2 0 0>
     }]
   }
   n24 = "Constraint - shoulderpad_l_MSTR (P)"!(ComponentPose=@n23, bAlphaBoolEnabled:bool=Get bIsSkyDiving(), ConstraintWeights_0=1) -> (Pose) {
    alphaInputType=Bool
    alphaBoolBlend {blendInTime=0.1, blendOption=Linear, blendOutTime=0.1}
    boneToModify {boneName=shoulderpad_l_MSTR}
    constraintSetup [{targetBone {boneName=upperarm_twist_01_l}}]
    constraintWeights [1]
   }
   n25 = RigidBody(ComponentPose=@n24) {
    bClampLinearTranslationLimitToRefPose=true
    bOverrideWorldGravity=true
    overridePhysicsAsset=&/M_MED_Garbage_Dyn_pelvis
    simulationSpace=BaseBoneSpace
    alphaBoolBlend {blendOption=Linear}
    baseBoneRef {boneName=pelvis}
    componentAppliedLinearAccClamp=<3000 3000 3000>
    componentLinearAccScale=<0.4 0.4 0.4>
    componentLinearVelScale=<0.4 0.4 0.4>
    overrideWorldGravity {z=-980}
   }
   n26 = RigidBody(ComponentPose=@n25) {
    bClampLinearTranslationLimitToRefPose=true
    bOverrideWorldGravity=true
    overridePhysicsAsset=&/M_MED_Garbage_Dyn_torso
    simulationSpace=BaseBoneSpace
    alphaBoolBlend {blendOption=Linear}
    baseBoneRef {boneName=spine_04}
    componentAppliedLinearAccClamp=<3000 3000 3000>
    componentLinearAccScale=<0.6 0.6 0.6>
    componentLinearVelScale=<0.6 0.6 0.6>
    overrideWorldGravity {z=-980}
   }
   n27 = RigidBody(ComponentPose=@n26) {
    bClampLinearTranslationLimitToRefPose=true
    bOverrideWorldGravity=true
    overridePhysicsAsset=&/M_MED_Garbage_Dyn_perf_banana
    simulationSpace=BaseBoneSpace
    alphaBoolBlend {blendOption=Linear}
    baseBoneRef {boneName=spine_04}
    componentAppliedLinearAccClamp=<3000 3000 3000>
    componentLinearAccScale=<0.6 0.6 0.6>
    componentLinearVelScale=<0.6 0.6 0.6>
    overrideWorldGravity {z=-980}
   }
   n28 = "Transform (Modify) Bone - Bone: tp_handle"(ComponentPose=@n27, Alpha=Get thigh_rot_l(), Translation="5.199,0.000,-0.283", Rotation="20.000,-30.000,-20.000", Scale="1,1,1") {
    rotationMode=BMM_Additive
    translationMode=BMM_Additive
    alphaBoolBlend {blendInTime=0.1, blendOption=Linear, blendOutTime=0.1}
    boneToModify {boneName=tp_handle}
    rotation=<rot 20 -30 -20>
    translation {x=5.199, z=-0.283}
   }
   n29 = Anim Dynamics - Bone: matches(ComponentPose=@n28, Alpha=1, ExternalForce="0,0,0") {
    angularDampingOverride=0.998
    angularSpringConstant=300
    bAngularSpring=true
    bOverrideAngularDamping=true
    bUseGravityOverride=true
    simulationSpace=RootRelative
    alphaBoolBlend {blendOption=Linear}
    boundBone {boneName=matches}
    componentLinearAccScale=<0.8 0.8 0.8>
    componentLinearVelScale=<0.8 0.8 0.8>
    gravityOverride {z=-980}
    physicsBodyDefinitions [{
      boundBone {boneName=matches}
      boxExtents=<7 12 12>
      constraintSetup {angularLimitsMax=<1 0 15>, angularLimitsMin=<-1 -10 -15>, angularTarget=<0 0 0>}
      localJointOffset=<1 0 0>
     }]
   }
   n30 = Anim Dynamics - Bone: can_fr(ComponentPose=@n29, Alpha=1, ExternalForce="0,0,0") {
    angularDampingOverride=0.998
    angularSpringConstant=300
    bAngularSpring=true
    bOverrideAngularDamping=true
    bUseGravityOverride=true
    simulationSpace=RootRelative
    alphaBoolBlend {blendOption=Linear}
    boundBone {boneName=can_fr}
    componentLinearAccScale=<0.8 0.8 0.8>
    componentLinearVelScale=<0.8 0.8 0.8>
    gravityOverride {z=-980}
    physicsBodyDefinitions [{
      boundBone {boneName=can_fr}
      boxExtents=<5 10 10>
      constraintSetup {angularLimitsMax=<1 0 15>, angularLimitsMin=<-1 -10 -15>, angularTarget=<0 0 0>}
      localJointOffset=<1 0 0>
     }]
   }
   n31 = Anim Dynamics - Bone: can_fr_lid(ComponentPose=@n30, Alpha=1, ExternalForce="0,0,0") {
    angularDampingOverride=0.998
    bOverrideAngularDamping=true
    simulationSpace=RootRelative
    alphaBoolBlend {blendOption=Linear}
    boundBone {boneName=can_fr_lid}
    componentLinearAccScale=<0.8 0.8 0.8>
    componentLinearVelScale=<0.8 0.8 0.8>
    physicsBodyDefinitions [{
      boundBone {boneName=can_fr_lid}
      boxExtents=<7 12 12>
      constraintSetup {angularLimitsMax=<1 1 10>, angularLimitsMin=<-1 -1 -10>, angularTarget=<0 0 0>}
      localJointOffset=<3 0 0>
     }]
   }
   n32 = Anim Dynamics - Bone: popcan(ComponentPose=@n31, Alpha=1, ExternalForce="0,0,0") {
    angularDampingOverride=0.998
    bOverrideAngularDamping=true
    simulationSpace=RootRelative
    alphaBoolBlend {blendOption=Linear}
    boundBone {boneName=popcan}
    componentLinearAccScale=<0.8 0.8 0.8>
    componentLinearVelScale=<0.8 0.8 0.8>
    physicsBodyDefinitions [{
      boundBone {boneName=popcan}
      boxExtents=<10 15 15>
      constraintSetup {angularLimitsMax=<1 1 15>, angularLimitsMin=<-1 -1 -15>, angularTarget=<0 0 0>}
      localJointOffset=<2 0 0>
     }]
   }
   n33 = Anim Dynamics - Bone: can_bk(ComponentPose=@n32, Alpha=1, ExternalForce="0,0,0") {
    angularDampingOverride=0.998
    angularSpringConstant=300
    bAngularSpring=true
    bOverrideAngularDamping=true
    bUseGravityOverride=true
    simulationSpace=RootRelative
    alphaBoolBlend {blendOption=Linear}
    boundBone {boneName=can_bk}
    componentLinearAccScale=<0.8 0.8 0.8>
    componentLinearVelScale=<0.8 0.8 0.8>
    gravityOverride {z=-980}
    physicsBodyDefinitions [{
      boundBone {boneName=can_bk}
      boxExtents=<5 10 10>
      constraintSetup {angularLimitsMax=<1 0 15>, angularLimitsMin=<-1 -10 -15>, angularTarget=<0 0 0>}
      localJointOffset=<1 0 0>
     }]
   }
   n34 = Anim Dynamics - Bone: can_bk_lid(ComponentPose=@n33, Alpha=1, ExternalForce="0,0,0") {
    angularDampingOverride=0.998
    bOverrideAngularDamping=true
    simulationSpace=RootRelative
    alphaBoolBlend {blendOption=Linear}
    boundBone {boneName=can_bk_lid}
    componentLinearAccScale=<0.8 0.8 0.8>
    componentLinearVelScale=<0.8 0.8 0.8>
    physicsBodyDefinitions [{
      boundBone {boneName=can_bk_lid}
      boxExtents=<7 12 12>
      constraintSetup {angularLimitsMax=<1 1 10>, angularLimitsMin=<-1 -1 -10>, angularTarget=<0 0 0>}
      localJointOffset=<3 0 0>
     }]
   }
   n35 = Anim Dynamics - Bone: tp_roll(ComponentPose=@n34, Alpha=1, ExternalForce="0,0,0") {
    angularDampingOverride=0.998
    bOverrideAngularDamping=true
    simulationSpace=RootRelative
    alphaBoolBlend {blendOption=Linear}
    boundBone {boneName=tp_roll}
    componentLinearAccScale=<0.4 0.4 0.4>
    componentLinearVelScale=<0.4 0.4 0.4>
    physicsBodyDefinitions [{
      boundBone {boneName=tp_roll}
      boxExtents=<15 20 20>
      constraintSetup {angularLimitsMax=<1 1 7>, angularLimitsMin=<-1 -1 -7>, angularTarget=<0 0 0>}
      localJointOffset=<2 0 0>
     }]
   }
   n36 = Anim Dynamics - Bone: tp_01(ComponentPose=@n35, Alpha=1, ExternalForce="0,0,0") {
    angularDampingOverride=0.998
    angularSpringConstant=300
    bAngularSpring=true
    bOverrideAngularDamping=true
    bUseGravityOverride=true
    simulationSpace=RootRelative
    alphaBoolBlend {blendOption=Linear}
    boundBone {boneName=tp_01}
    componentLinearAccScale=<0.8 0.8 0.8>
    componentLinearVelScale=<0.8 0.8 0.8>
    gravityOverride {z=-980}
    physicsBodyDefinitions [{
      collisionType=CustomSphere
      sphereCollisionRadius=5
      boundBone {boneName=tp_01}
      boxExtents=<10 15 15>
      constraintSetup {angularLimitsMax=<1 0 10>, angularLimitsMin=<-1 -10 -10>, angularTarget=<0 0 0>}
      localJointOffset=<1 0 0>
     }]
    sphericalLimits [{limitRadius=9, drivingBone {boneName=thigh_l}, sphereLocalOffset=<-5 0 0>}]
   }
   n37 = Anim Dynamics - Bone: tp_02(ComponentPose=@n36, Alpha=1, ExternalForce="0,0,0") {
    angularDampingOverride=0.998
    angularSpringConstant=300
    bAngularSpring=true
    bOverrideAngularDamping=true
    bUseGravityOverride=true
    bUseSphericalLimits=true
    simulationSpace=RootRelative
    alphaBoolBlend {blendOption=Linear}
    boundBone {boneName=tp_02}
    componentLinearAccScale=<0.8 0.8 0.8>
    componentLinearVelScale=<0.8 0.8 0.8>
    gravityOverride {z=-980}
    physicsBodyDefinitions [{
      collisionType=CustomSphere
      sphereCollisionRadius=5
      boundBone {boneName=tp_02}
      boxExtents=<10 15 15>
      constraintSetup {angularLimitsMax=<1 0 10>, angularLimitsMin=<-1 -10 -10>, angularTarget=<0 0 0>}
      localJointOffset=<1 0 0>
     }]
    sphericalLimits [{limitRadius=10, drivingBone {boneName=thigh_twist_01_l}, sphereLocalOffset=<5 0 0>}]
   }
   n38 = Get is_wearing_westsausage() -> (is_wearing_westsausage:bool)
   n39 = "Constraint - dyn_tire (P)"!(ComponentPose=@n37, bAlphaBoolEnabled:bool=@n38, ConstraintWeights_0=1) -> (Pose) {
    alphaInputType=Bool
    alphaBoolBlend {blendOption=Linear}
    boneToModify {boneName=dyn_tire}
    constraintSetup [{targetBone {boneName=dyn_spine_04}}]
    constraintWeights [1]
   }
   n40 = "Constraint - shoulderpad_r (P)"!(ComponentPose=@n39, bAlphaBoolEnabled:bool=@n38, ConstraintWeights_0=1) -> (Pose) {
    alphaInputType=Bool
    alphaBoolBlend {blendOption=Linear}
    boneToModify {boneName=shoulderpad_r}
    constraintSetup [{targetBone {boneName=alt_upperarm_twist_01_r}}]
    constraintWeights [1]
   }
   Output Pose(Result=Component To Local(ComponentPose=@n40)) {layerGroup=None}
   Sequence Player 'RunningAround_v01'() {playRate=0.1, sequence=&/RunningAround_v01, startPosition=1}
   Break Fort Anim Input Skydiving External Force(FortAnimInput_SkydivingExternalForce=Get SkydivingExternalForceInput()) -> (
    bUseSkydivingVectorForce=~
    SkydivingVectorForce=~
    HeadToPelvisDirection=~
    FloatingMultiplier=~
    FloatingAdditive=~
    DivingMultiplier=~
    DivingAdditive=~
    ParachutingMultiplier=~
    ParachutingAdditive=~
   )
   Get IsCrouchingAnim() -> (IsCrouchingAnim:bool)
   Get left arm rot()
  ]}
 EventGraph {nodes [
   Event BlueprintUpdateAnimation()
   > n1 = Sequence!(execute) -> (then_0, then_1, then_2:Exec)
   > n2 = Branch(Condition=Get bIsSkyDiving())
   > Set left arm rot(left arm rot:double=0) -> (Output_Get:double)
   Set left arm rot(
    execute=@n2.else
    left arm rot=GetAlphaFromReferenceTransformDelta(SkeletalMeshComponent=Get Skel_Mesh(), BoneName=upperarm_l, BaseName=spine_05, InputRange=(X=0,Y=90), TwistAxis="1,0,0")
   )
   n3 = GetOwnerSkeletalMeshComponent()
   n4 = BreakRotator(InRot=BreakTransform(InTransform=GetDeltaTransformFromRefPose(self=@n3, BoneName=upperarm_l)).Rotation)
   > Set LeftShoulder_Rot(LeftShoulder_Rot:Rotator=MakeRotator(Pitch=MapRangeClamped(Value=@n4.Yaw, InRangeA=40, InRangeB=80, OutRangeA=20, OutRangeB=40))) -> (Output_Get:Rotator)
   n5 = "Not Equal (Boolean)"(A=Get bIsMoving2D(), B=true) -> (ReturnValue=~)
   n6 = Get bIsCrouching()
   n7 = Get CurrentLOD()
   > Set IsCrouchingAnim(IsCrouchingAnim:bool=AND Boolean(A=AND Boolean(A=@n6, B=@n5), B="Equal (Integer)"(A=@n7, B=1))) -> (Output_Get:bool)
   n8 = Get Skel_Mesh()
   Set thigh_rot_l(
    execute=@n1.then_2
    thigh_rot_l=GetAlphaFromReferenceTransformDelta(SkeletalMeshComponent=@n8, BoneName=thigh_l, BaseName=pelvis, InputRange=(X=-20,Y=-115), TwistAxis="1,0,0")
   )
   > Set thigh_rot_r(thigh_rot_r=GetAlphaFromReferenceTransformDelta(SkeletalMeshComponent=@n8, BoneName=thigh_r, BaseName=pelvis, InputRange=(X=-20,Y=-100), TwistAxis="1,0,0"))
   n9 = Cast To FortPlayerPawn(execute=@n1.then_1, Object=GetOwningActor())
   > Set As Fort Player Pawn(As Fort Player Pawn:Fort Player Pawn Object Reference=@n9) -> (Output_Get:Fort Player Pawn Object Reference)
   > n10 = Get!(execute:Exec) -> (then:Exec, else:Exec, As Fort Player Pawn:Fort Player Pawn Object Reference)
   > Set is_wearing_westsausage(is_wearing_westsausage:bool=HasMatchingGameplayTag!(self:Gameplay Tag Asset Interface Interface=@n10, TagToCheck:Gameplay Tag Structure=(TagName="Granted.Athena.HandsAreReplaced")) -> (ReturnValue:bool)) -> (Output_Get:bool)
   Event OnParachuteOpened()
   > n11 = Sequence!(execute) -> (then_0, then_1)
   > n12 = Delay(Duration=1.5)
   ResetDynamics(execute=[@n17.then, @n12.then, @n19.then_0, @n15.then], InTeleportType=TeleportPhysics)
   > ResetDynamics(InTeleportType=ResetPhysics)
   Set ShoulderpadRightRot_b(execute=@n11.then_1, ShoulderpadRightRot_b:Rotator="-110, 0,13.82") -> (Output_Get:Rotator)
   Event BlueprintInitializeAnimation()
   > Set Skel_Mesh(Skel_Mesh=GetOwnerSkeletalMeshComponent())
   n13 = Event OnBeginTargeting() {is=title K2Node_Event}
   n14 = Get bIsCrouching()
   n15 = Branch(execute=[@n13.then, @n18.then], Condition=AND Boolean(A=Get bIsTargeting(), B=NOT Boolean(A=@n14)))
   Event OnEndSkydiving()
   > n16 = Sequence()
   > n17 = Delay(Duration=0.2)
   Set SkydivingAlpha(execute=@n16.then_1, SkydivingAlpha:float=0) -> (Output_Get:float)
   > Set SkydivingInvertedAlpha(SkydivingInvertedAlpha:float=1) -> (Output_Get:float)
   > Set SkydivingAlpha_b(SkydivingAlpha_b:double=0) -> (Output_Get:double)
   > Set SkydivingInvertedAlpha_b(SkydivingInvertedAlpha_b:double=1) -> (Output_Get:double)
   n18 = Event OnEndCrouching()
   n19 = Sequence()
   Set ShoulderpadRightRot_b(execute=@n19.then_1, ShoulderpadRightRot_b="-50.726, 0,13.82")
   > Set SkydivingAlpha_b(SkydivingAlpha_b=1)
   > Set SkydivingInvertedAlpha_b(SkydivingInvertedAlpha_b=0)
   Event OnBeginSkydiving()
  ]}
]
variables [
 As Fort Player Pawn {type=FortPlayerPawn}
 is_wearing_westsausage {type=boolean}
 IsCrouchingAnim {type=boolean}
 IsMovingAnim {type=boolean}
 left arm rot {type=number}
 LeftShoulder_Rot {type=Rotator}
 ShoulderpadRightRot_b {type=Rotator, default=<rot -50.726 0 13.82>}
 Skel_Mesh {type=SkeletalMeshComponent}
 SkydivingAlpha_b {type=number}
 SkydivingInvertedAlpha_b {type=number, default=1}
 thigh_rot_l {type=number}
 thigh_rot_r {type=number}
]
